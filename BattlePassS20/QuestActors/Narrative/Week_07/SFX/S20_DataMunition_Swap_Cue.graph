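# Sound cue for the data munition swap effect in a Season 20 Week 7 narrative quest, assigned to the quests sound class.
# A single wave player feeds a modulator that adds slight random pitch variation.
# It references the S20 quest attenuation asset but also enables override attenuation, with logarithmic falloff, low-pass filtering over distance and listener-focus scaling.
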SoundCue /BattlePassS20/QuestActors/Narrative/Week_07/SFX/S20_DataMunition_Swap_Cue
root=Modulator_0
nodes [
 {id=Modulator_0, children [WavePlayer_0], properties {pitchMax=1.02, pitchMin=0.98, volumeMax=1}}
 {id=WavePlayer_0, properties {soundWaveAssetPtr=&/Munition_Swap_01}}
]
properties {
 attenuationSettings=/BattlePassS20/Audio/VO/S20_Quest_Att
 bOverrideAttenuation=true
 maxDistance=9250
 soundClassObject=&Fort_Quests
 attenuationOverrides {
  bAttenuateWithLPF=true
  bEnableListenerFocus=true
  distanceAlgorithm=Logarithmic
  falloffDistance=7750
  lPFFrequencyAtMax=12000
  nonFocusAzimuth=45
  nonFocusDistanceScale=0.5
  nonFocusVolumeAttenuation=0.2
  stereoSpread=1000
  attenuationShapeExtents {x=1500}
 }
}
