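# Sound cue for the completion of the server hack in Battle Pass Season 20, Week 7, Quest 4.
# A single wave player feeds the HackServer_Complete sound wave straight to the output, with no randomization or mixing.
# Playback uses the quest's shared hack attenuation with a limited audible range and routes through the Fort_Quests sound class.
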SoundCue /BattlePassS20/QuestActors/Narrative/Week_07/SFX/S20_W07_Q04_HackServer_Complete_Cue
root=WavePlayer_0
nodes [{id=WavePlayer_0, properties {soundWaveAssetPtr=&/S20_W07_Q04_HackServer_Complete}}]
properties {
 attenuationSettings=&/S20_W07_Q04_Hack_Atten
 maxDistance=1300
 soundClassObject=&Fort_Quests
 volumeMultiplier=1
}
