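# Sound cue for switching on the hack server in a Season 20 Battle Pass narrative quest (Week 7, Quest 4).
# A Random node picks one of three equally weighted wave variations and feeds a Modulator that fixes volume, leaving pitch modulation at its default.
# The cue plays at a lowered pitch multiplier.
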SoundCue /BattlePassS20/QuestActors/Narrative/Week_07/SFX/S20_W07_Q04_HackServer_On
root=Modulator_0
nodes [
 {id=Modulator_0, children [Random_0], properties {volumeMax=1, volumeMin=1}}
 {id=Random_0, children [WavePlayer_0,WavePlayer_1,WavePlayer_2], properties {weights [1,1,1]}}
 {id=WavePlayer_0, properties {soundWaveAssetPtr=&/S20_W07_Q04_HackServer_On_01}}
 {id=WavePlayer_1, properties {soundWaveAssetPtr=&/S20_W07_Q04_HackServer_On_02}}
 {id=WavePlayer_2, properties {soundWaveAssetPtr=&/S20_W07_Q04_HackServer_On_03}}
]
properties {maxDistance=2097152, pitchMultiplier=0.7, volumeMultiplier=1}
