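# Looping sound cue for the server ambience in the Season 20 Week 7 Quest 4 narrative quest.
# A looping wave player feeds an enveloper that ramps volume from silence to full over a short fade-in.
# It uses the quest's hack attenuation asset and plays in the Fort_Quests sound class.
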SoundCue /BattlePassS20/QuestActors/Narrative/Week_07/SFX/S20_W07_Q04_Server_LP_Cue
root=Enveloper_0
nodes [
 {
  id=Enveloper_0
  children [WavePlayer_0]
  properties {pitchCurve {editorCurveData {keys [{time=0}]}}, volumeCurve {editorCurveData {keys [<at 0 0>,<at 0.2 1>]}}}
 }
 {id=WavePlayer_0, properties {bLooping=true, soundWaveAssetPtr=&/S20_W07_Q04_Server_LP}}
]
properties {
 attenuationSettings=&/S20_W07_Q04_Hack_Atten
 maxDistance=1300
 soundClassObject=&Fort_Quests
 volumeMultiplier=1
}
